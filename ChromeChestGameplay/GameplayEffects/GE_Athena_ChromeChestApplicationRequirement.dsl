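; Gameplay effect application requirement that allows the Chrome Chest loot-changing effect only on building containers eligible for it.
; CanApplyGameplayEffect casts the ability system component's owner to a BuildingContainer, then finds ChromeChestGameStateComponent on the Athena game state.
; It returns the validity result from that component's GetChromeChestOverrideInfo for the container.
; A Sequence node supplies a false fallback whenever that chain does not reach its return.
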(fn CanApplyGameplayEffect
  (params (GameplayEffect :type Gameplay%20Effect%20Object%20Reference) (Spec :type Gameplay%20Effect%20Spec%20Structure%20%28by%20ref%29) (ASC :type Ability%20System%20Component%20Object%20Reference))
  ;[0.04005 0.2797 0.3073] Check to see if the owner is a valid container that can accept the loot changing GE.
  (Utilities|FlowControl|Sequence
    (:then_0
      (Utilities|IsValid ASC
        (:Is%20Valid
          (bind _asbuilding_container (Utilities|Casting|CastToBuildingContainer (Components|GetOwner ASC)))
          (Utilities|IsValid (Fortnite|Athena|GameState|GetGameStateAthenafromContext _asbuilding_container)
            (:Is%20Valid
              (Utilities|IsValid (Actor|GetComponentbyClass (Fortnite|Athena|GameState|GetGameStateAthenafromContext _asbuilding_container) "/ChromeChestGameplay/Gameplay/ChromeChestGameStateComponent.ChromeChestGameStateComponent_C")
                (:Is%20Valid
                  (bind _bvalid (Class|ChromeChestGameStateComponent|GetChromeChestOverrideInfo (Actor|GetComponentbyClass (Fortnite|Athena|GameState|GetGameStateAthenafromContext _asbuilding_container) "/ChromeChestGameplay/Gameplay/ChromeChestGameStateComponent.ChromeChestGameStateComponent_C") _asbuilding_container))
                  (return _bvalid))))))))
    (:then_1
      (return false))))
